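# Physics asset for the commando Shark Hat cosmetic mesh, giving the hat simple collision on the character's upper body.
# It has two capsule bodies, one on head1 and one on spine_05, each offset and rotated from its bone.
# A single constraint joins head1 to spine_05 with matching swing and twist limit angles.
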PhysicsAsset /BRCosmetics/Accessories/Hats/M_MED_Shark_Hat/Meshes/M_MED_commando_SharkHat_PhysicsAsset
bodies {
 head1 [{shapeType=Capsule, length=19.1805, radius=13.7033, center=<6.291 4.6792 -1.1969>, rotation=<rot 0 -59.7876 -81.8455>}]
 spine_05 [{shapeType=Capsule, length=13.4022, radius=12.5257, center=<12.8894 0.0316 -0.7056>, rotation=<rot 0 21.6665 -75.5208>}]
}
constraints [{bone1Name=head1, bone2Name=spine_05, swing1LimitDegrees=45, swing2LimitDegrees=45, twistLimitDegrees=45}]
